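; Interactable test actor base for Clash that spawns a skydrop pod on demand when interacted with.
; Each interaction destroys all actors of the pod container class, spawns the skydrop class at an unrotated offset from the actor, then disables interaction until a cooldown timer re-enables it.
; A downward line trace places the drop cue sound at the ground hit, falling back to the spawn point.
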(event EventBeginPlay)

(event Collision|EventActorBeginOverlap
  (params (OtherActor :type Actor%20Object%20Reference)))

(event EventTick
  (params (DeltaSeconds :type Float%20%28single-precision%29)))

(event Interaction|EventBlueprintOnInteract
  (params (InteractingPawn :type Fort%20Pawn%20Object%20Reference) (InteractionBeingAttempted :type EInteractionBeingAttempted%20Enum :default "FirstInteraction"))
  (CallFunction|CleanUpOldPods)
  (CallFunction|SpawnDropPod)
  (Variables|Default|SetCanInteract false)
  (Utilities|Time|SetTimerbyEvent (delegate Custom|EnableInteraction) (Variables|Default|GetInteractionCooldown)))

(event Custom|EnableInteraction
  (Variables|Default|SetCanInteract true))

(fn BlueprintCanInteract
  (params (InteractingPawn :type Fort%20Pawn%20Object%20Reference) (InteractionBeingAttempted :type EInteractionBeingAttempted%20Enum :default "FirstInteraction") (InteractionType :type TInteractionType%20Enum :default "IT_NoInteraction"))
  (return (Variables|Default|GetCanInteract)))

(fn CleanUpOldPods
  (bind _outactors (Actor|GetAllActorsOfClass (Variables|Default|GetPodContainerClass)))
  (for _array_element _outactors
    (Actor|DestroyActor _array_element)))

(fn NewFunction_0)

(fn PlaySoundAtGroundDestination
  (params (SpawnPoint :type Vector))
  (bind (_outhit _returnvalue) (Collision|LineTraceByChannel SpawnPoint (- SpawnPoint "0, 0,10000.000000") "TraceTypeQuery7"))
  (Audio|PlaySoundatLocation (Variables|Default|GetDropCueSound) (select _returnvalue (.Location (Collision|BreakHitResult _outhit)) (Variables|GetSpawnPoint))))

(fn SpawnDropPod
  (bind _returnvalue (Transformation|GetActorLocation))
  (bind _podspawnpoint (Variables|Default|GetPodSpawnPoint))
  (bind _returnvalue_2 (+ _returnvalue _podspawnpoint))
  (Game|SpawnActorfromClass (Variables|Default|GetPodSkydropClass) (Math|Transform|MakeTransform _returnvalue_2) "Undefined" "SelectDefaultAtRuntime" 0 0)
  (CallFunction|PlaySoundatGroundDestination _returnvalue_2)
  (Development|PrintString "Pod spawned"))

(fn ConstructionScript)
